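# MF_SmoothStep is a material function in the Colortrip vehicle boost folder that computes a Hermite smoothstep of a scalar Value between Min and Max.
# It remaps Value to a saturated 0–1 factor t by dividing (Value − Min) by (Max − Min), then evaluates (3 − 2t)·t·t.
# The scalar result is appended into three identical channels, so the output is a float3 rather than a scalar.
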MaterialFunction /VehicleCosmetics/Boosts/Boost_Colortrip/Materials/MF_SmoothStep
graph {nodes [
  n1 = Constant() {r=3}
  n2 = FunctionInput() {inputName=Value, inputType=FunctionInput_Scalar, sortPriority=1, previewValue=<1 1 1 1>}
  n3 = FunctionInput() {inputName=Min, inputType=FunctionInput_Scalar, sortPriority=2, previewValue=<0.1 0.1 0.1 1>}
  n4 = FunctionInput() {inputName=Max, inputType=FunctionInput_Scalar, sortPriority=3, previewValue=<0.9 0.9 0.9 0.9>}
  MaterialExpressionConstantClamp_4 = Saturate(None=Divide(A=Subtract(A=@n2, B=@n3), B=Subtract(A=@n4, B=@n3)))
  n5 = Multiply(A=@MaterialExpressionConstantClamp_4, B=Constant() {r=2})
  n6 = Multiply(A=Multiply(A=Subtract(A=@n1, B=@n5), B=@MaterialExpressionConstantClamp_4), B=@MaterialExpressionConstantClamp_4)
  FunctionOutput(None=AppendVector(A=AppendVector(A=@n6, B=@n6), B=@n6))
 ]}
