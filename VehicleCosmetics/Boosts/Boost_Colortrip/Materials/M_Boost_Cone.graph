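# Unlit, masked, two-sided material for the cone of the Colortrip vehicle boost, built for Niagara mesh particles.
# Two panning samples of a bicolor texture, offset by particle TrailOffset and distorted by noise, layer colors under a bright UV-gradient center tint.
# A Painteds parameter blends desaturated to full color before the emissive adjustment function.
# Separate panning cloud noise and a fire mask warp a sharpened UV gradient, alongside a hard UV1 cutoff.
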Material /VehicleCosmetics/Boosts/Boost_Colortrip/Materials/M_Boost_Cone
graph {
 properties {
  bAutomaticallySetUsageInEditor=false
  bUsedWithNiagaraMeshParticles=true
  blendMode=BLEND_Masked
  shadingModel=MSM_Unlit
  twoSided=true
  usedShadingModels=Unlit
 }
 nodes [
  n1 = Constant() {r=3}
  n2 = Multiply(A=ScalarParameter() {defaultValue=1, parameterName=Speed}, B=Time())
  n3 = TextureSample(UVs=Panner(Coordinate=TextureCoordinate() {uTiling=2, vTiling=1.2}, Time=@n2) {speedX=-0.25, speedY=0.35}) {samplerType=SAMPLERTYPE_LinearColor, texture=/VehicleCosmetics/Boosts/Boost_Colortrip/Textures/T_VCVFX_Boost_colortrip_CloudNoise_N}
  n4 = TextureSample(UVs=Panner(Coordinate=TextureCoordinate() {uTiling=2, vTiling=1.6}, Time=@n2) {speedX=0.2, speedY=0.5}) {samplerType=SAMPLERTYPE_LinearColor, texture=&T_VCVFX_Boost_colortrip_CloudNoise_N}
  n5 = Add(A=@n3.G, B=@n4.G)
  n6 = Add(A=@n3.R, B=@n4.R)
  n7 = Panner(Coordinate=TextureCoordinate()) {speedX=0.4}
  MaterialExpressionConstantClamp_6 = Saturate(None=Multiply(
    A=AppendVector(A=@n5, B=@n6)
    B=TextureSample(UVs=@n7) {texture=/VehicleCosmetics/Boosts/Boost_Colortrip/Textures/T_VCVFX_Boost_colortrip_Fire_Mask_A}.R
   ))
  n8 = ComponentMask(None=Add(A=TextureCoordinate(), B=Multiply(A=@n1, B=@MaterialExpressionConstantClamp_6))) {g=true}
  n9 = ScalarParameter() {defaultValue=4, parameterName=GradientAmount}
  MaterialExpressionConstantClamp_4 = Saturate(None=Multiply(A=Power(Base=@n8, Exp=@n9), B=ScalarParameter() {defaultValue=2, parameterName=GradientSharpness}))
  n10 = TextureCoordinate() {coordinateIndex=1}
  MaterialExpressionConstantClamp_7 = Saturate(None=ConstantBiasScale(None=ComponentMask(None=@n10) {r=true}) {bias=-0.5, scale=100})
  n11 = VectorParameter() {parameterName=InnerColor, defaultValue=<rgba 0 0.8 0.5 1>}
  n12 = TextureCoordinate()
  n13 = Multiply(
   A=Power(Base=ComponentMask(None=@n12) {g=true}, Exp=ScalarParameter() {defaultValue=10, parameterName=PoweCenter})
   B=ScalarParameter() {defaultValue=3, parameterName=BriCenter}
  )
  n14 = ScalarParameter() {defaultValue=1, parameterName=Brightness}
  n15 = DynamicParameter() {paramNames [TrailOffset,Param2,Param3,Param4]}
  n16 = TextureCoordinate()
  n17 = AppendVector(
   A=ScalarParameter() {defaultValue=3, group=MainTexture, parameterName=TilingU}
   B=ScalarParameter() {defaultValue=2, group=MainTexture, parameterName=TilingV}
  )
  n18 = AppendVector(A=Constant(), B=Constant() {r=-0.08})
  n19 = TextureCoordinate() {vTiling=0.2}
  n20 = TextureSample(UVs=Panner(Coordinate=Add(A=@n18, B=@n19)) {speedX=0.001, speedY=-0.01}) {texture=/DealershipSharedLibrary/FX/Textures/Noise/T_VCVFX_Alpine_Infinite_RGBNoise01}
  n21 = Add(
   A=Add(A=AppendVector(A=@n15.TrailOffset, B=@n15.TrailOffset), B=Multiply(A=@n16, B=@n17))
   B=Multiply(A=ComponentMask(None=@n20.RGB) {g=true, r=true}, B=ScalarParameter() {defaultValue=0.2, group=MainTexture, parameterName=StreghtDeform})
  )
  n22 = TextureSampleParameter2D(UVs=Panner(Coordinate=@n21) {speedX=0.3, speedY=1}) {group=MainTexture, parameterName=MainTexture, texture=/VehicleCosmetics/Boosts/Boost_Colortrip/Textures/T_VCVFX_Boost_Bicolor_RGB}
  n23 = TextureSampleParameter2D(UVs=Panner(Coordinate=@n21) {speedX=0.2, speedY=0.8}) {group=MainTexture, parameterName=MainTexture, texture=&T_VCVFX_Boost_Bicolor_RGB}
  n24 = LinearInterpolate(A=LinearInterpolate(A=Constant(), B=Constant3Vector() {constant=<rgba 0.8994 0.7969 0.3993 0>}, Alpha=@n22.R), B=@n11.RGB, Alpha=@n23.G)
  n25 = ScalarParameter() {defaultValue=2, parameterName=InnerBri}
  n26 = LinearInterpolate(
   A=Add(A=Multiply(A=@n11.RGB, B=@n13), B=Multiply(A=@n14, B=@n24))
   B=Multiply(A=Multiply(A=@n11.RGB, B=@n25), B=Constant() {r=1})
   Alpha=Constant()
  )
  MaterialFunctionCall(Emissive=LinearInterpolate(A=Desaturation(None=@n26), B=@n26, Alpha=ScalarParameter() {parameterName=Painteds})) {materialFunction=&MF_FX_EmissiveAdjustments}
 ]
}
